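# Physics asset for the Arms mesh of the Gingerbread Man Builder pet skin, defining its upper-body physics bodies.
# Single small capsules cover pelvis, spine, neck, head, both arm chains, pack, bowtie and Dyn_Helmet.
# Narrow angular limits keep elbows, wrists, head, helmet and spine fairly stiff, with looser shoulders and bowtie.
# Clavicle-to-spine, neck-to-spine and pelvis-to-pack joints leave swing and twist free.
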PhysicsAsset /BRCosmetics/Characters/Pets/Gingerbread/Skins/Gingerbread_Man_Builder/Mesh/M_MED_Gingerbread_Man_Builder_Arms_PhysicsAsset
bodies {
 Dyn_Helmet [{shapeType=Capsule, length=3, radius=3, center=<2.435 0 0>, rotation=<rot 0 0 90>}]
 bowtie [{shapeType=Capsule, length=1, radius=1, center=<-0.0297 -0.0111 -0.1779>, rotation=<rot 0 -81.9335 -89.2904>}]
 clavicle_l [{shapeType=Capsule, length=0.5, radius=0.5}]
 clavicle_r [{shapeType=Capsule, length=0.5, radius=0.5}]
 hand_l [{shapeType=Capsule, length=1, radius=1, center=<1.2079 -0.0313 -0.1047>, rotation=<rot 12.1596 -90 15.2242>}]
 hand_r [{shapeType=Capsule, length=1, radius=1, center=<-1.147 0.0252 0.1455>, rotation=<rot 4.1923 -90 14.1115>}]
 head [{shapeType=Capsule, length=3, radius=3, center=<2.2134 0 0>}]
 lowerarm_l [{shapeType=Capsule, length=2, radius=1, center=<1 0 0>, rotation=<rot 90 0 0>}]
 lowerarm_r [{shapeType=Capsule, length=2, radius=1, center=<-1 0 0>, rotation=<rot 90 0 0>}]
 neck_02 [{shapeType=Capsule, length=0.5, radius=0.5}]
 pack [{shapeType=Capsule, length=0.505, radius=0.505, rotation=<rot 0 -90 0>}]
 pelvis [{shapeType=Capsule, length=1, radius=1}]
 spine_03 [{shapeType=Capsule, length=1, radius=1}]
 upperarm_l [{shapeType=Capsule, length=2, radius=1, center=<1 0 0>, rotation=<rot 90 0 0>}]
 upperarm_r [{shapeType=Capsule, length=2, radius=1, center=<-1 0 0>, rotation=<rot 90 0 0>}]
}
constraints [
 {bone1Name=lowerarm_l, bone2Name=upperarm_l, swing1LimitDegrees=30, swing2LimitDegrees=1, twistLimitDegrees=1}
 {bone1Name=hand_l, bone2Name=lowerarm_l, swing1LimitDegrees=10, swing2LimitDegrees=1, twistLimitDegrees=1}
 {bone1Name=lowerarm_r, bone2Name=upperarm_r, swing1LimitDegrees=20, swing2LimitDegrees=1, twistLimitDegrees=1}
 {bone1Name=hand_r, bone2Name=lowerarm_r, swing1LimitDegrees=10, swing2LimitDegrees=1, twistLimitDegrees=1}
 {bone1Name=Dyn_Helmet, bone2Name=head, swing1LimitDegrees=5, swing2LimitDegrees=5, twistLimitDegrees=1}
 {bone1Name=bowtie, bone2Name=spine_03, swing1LimitDegrees=20, swing2LimitDegrees=20, twistLimitDegrees=1}
 {bone1Name=upperarm_l, bone2Name=clavicle_l, swing1LimitDegrees=10, swing2LimitDegrees=60, twistLimitDegrees=35}
 {bone1Name=upperarm_r, bone2Name=clavicle_r, swing1LimitDegrees=10, swing2LimitDegrees=60, twistLimitDegrees=35}
 {bone1Name=head, bone2Name=neck_02, swing1LimitDegrees=10, swing2LimitDegrees=10, twistLimitDegrees=5}
 {
  bone1Name=clavicle_r
  bone2Name=spine_03
  swing1LimitDegrees=45
  swing1Motion=Free
  swing2LimitDegrees=45
  swing2Motion=Free
  twistLimitDegrees=45
  twistMotion=Free
 }
 {
  bone1Name=clavicle_l
  bone2Name=spine_03
  swing1LimitDegrees=45
  swing1Motion=Free
  swing2LimitDegrees=45
  swing2Motion=Free
  twistLimitDegrees=45
  twistMotion=Free
 }
 {
  bone1Name=pelvis
  bone2Name=pack
  swing1LimitDegrees=45
  swing1Motion=Free
  swing2LimitDegrees=45
  swing2Motion=Free
  twistLimitDegrees=45
  twistMotion=Free
 }
 {bone1Name=spine_03, bone2Name=pelvis, swing1LimitDegrees=20, swing2LimitDegrees=1, twistLimitDegrees=1}
 {
  bone1Name=neck_02
  bone2Name=spine_03
  swing1LimitDegrees=45
  swing1Motion=Free
  swing2LimitDegrees=45
  swing2Motion=Free
  twistLimitDegrees=45
  twistMotion=Free
 }
]
