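# Physics asset for the legs mesh of the Gingerbread Man Builder pet skin, defining its simplified collision bodies and joint limits.
# Capsules cover the pack, pelvis, and both thighs and calves, with the leg shapes mirrored left to right.
# Knee constraints act as tight near-hinges and hip constraints allow a wider limited swing and twist.
# The pelvis-to-pack constraint leaves all angular motion free.
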PhysicsAsset /BRCosmetics/Characters/Pets/Gingerbread/Skins/Gingerbread_Man_Builder/Mesh/M_MED_Gingerbread_Man_Builder_Legs_PhysicsAsset
bodies {
 calf_l [{shapeType=Capsule, length=2, radius=1, center=<-2 0 0>, rotation=<rot 90 0 0>}]
 calf_r [{shapeType=Capsule, length=2, radius=1, center=<2 0 0>, rotation=<rot 90 0 0>}]
 pack [{shapeType=Capsule, length=0.505, radius=0.505, rotation=<rot 0 -90 0>}]
 pelvis [{shapeType=Capsule, length=1, radius=1}]
 thigh_l [{shapeType=Capsule, length=3, radius=1.5, center=<-2 0 0>, rotation=<rot 90 0 0>}]
 thigh_r [{shapeType=Capsule, length=3, radius=1.5, center=<2 0 0>, rotation=<rot 90 0 0>}]
}
constraints [
 {bone1Name=calf_l, bone2Name=thigh_l, swing1LimitDegrees=15, swing2LimitDegrees=1, twistLimitDegrees=1}
 {bone1Name=calf_r, bone2Name=thigh_r, swing1LimitDegrees=15, swing2LimitDegrees=1, twistLimitDegrees=1}
 {bone1Name=thigh_l, bone2Name=pelvis, swing1LimitDegrees=50, swing2LimitDegrees=15, twistLimitDegrees=15}
 {bone1Name=thigh_r, bone2Name=pelvis, swing1LimitDegrees=50, swing2LimitDegrees=15, twistLimitDegrees=15}
 {
  bone1Name=pelvis
  bone2Name=pack
  swing1LimitDegrees=45
  swing1Motion=Free
  swing2LimitDegrees=45
  swing2Motion=Free
  twistLimitDegrees=45
  twistMotion=Free
 }
]
